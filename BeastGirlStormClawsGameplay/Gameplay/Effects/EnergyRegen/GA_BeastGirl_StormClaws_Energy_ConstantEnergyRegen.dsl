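; Passive gameplay ability for BeastGirl StormClaws that grants the player constant energy regeneration while active.
; On activate it tells the avatar pawn's Fort energy component to start adding energy at a scalable per-second rate under a source ID.
; On end it validates that component and stops adding energy for the same source ID.
; Both events print debug strings, with the end event reporting whether the pawn has the ConstantRegen granted tag.
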(event Ability|EventActivateAbility
  ;[1 1 1] On Activate begin adding energy. On end remove the added energy.
  (Development|PrintString "EnergyRegenGranted")
  (Energy|StartAddingEnergy (Actor|GetComponentbyClass (CallFunction|GetAvatarasFortPlayerPawn) "/Script/EnergyRuntime.FortComponent_Energy") (ScalableFloat|GetValueatLevel (Variables|Default|GetRow_EnergyPerSecond)) (Variables|Default|GetT_EnergySourceID) 0))

(event Ability|EventOnEndAbility
  (params (bWasCancelled :type Boolean))
  (bind _returnvalue (CallFunction|GetAvatarasFortPlayerPawn))
  (bind _returnvalue_2 (Actor|GetComponentbyClass _returnvalue "/Script/EnergyRuntime.FortComponent_Energy"))
  ;[1 1 1] On Activate begin adding energy. On end remove the added energy.
  (Development|PrintString (Utilities|String|ToString(Boolean) (GameplayTags|HasMatchingGameplayTag (Ability|GetActivatingPawn) "(TagName=\"Granted.Athena.Energy.ConstantRegen\")")))
  (Utilities|IsValid _returnvalue_2
    (:Is%20Valid
      (Energy|StopAddingEnergy _returnvalue_2 (Variables|Default|GetT_EnergySourceID) 0))))

(event Ability|EventFailedToActivatePassiveAbility
  (params (ActorInfo :type Gameplay%20Ability%20Actor%20Info%20Structure))
  ;[1 1 1] Must implement this given it is a passive ability.
  (|CallParentFunction))
